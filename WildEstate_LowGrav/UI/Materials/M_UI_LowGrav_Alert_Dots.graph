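# Translucent UI material that draws three small animated dots for a low-gravity alert graphic.
# Each dot is an MF_UI_Circle call with offset UVs, bobbing vertically through a sine of Time with its own speed and amplitude.
# Summed fills and outlines feed named reroutes; opacity is a lerp of reroute values scaled by an Opacity parameter.
# Final color multiplies a rerouted value by the BackroundColor_4 tint parameter.
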Material /WildEstate_LowGrav/UI/Materials/M_UI_LowGrav_Alert_Dots
graph {
 properties {bAutomaticallySetUsageInEditor=false, blendMode=BLEND_Translucent, materialDomain=MD_UI, usedShadingModels=Default Lit}
 nodes [
  NamedRerouteDeclaration(None=Multiply(A=NamedRerouteUsage(), B=VectorParameter() {parameterName=BackroundColor_4, defaultValue=<rgba 0.4325 0.151 1 1>}.RGB)) {name=Final Color, nodeColor=<rgba 0 0.7891 1 1>}
  n1 = TextureCoordinate()
  n2 = ComponentMask(None=@n1) {r=true}
  n3 = Add(A=ComponentMask(None=@n1) {g=true}) {constB=-0.2}
  n4 = Time()
  n5 = LinearInterpolate(Alpha=Sine(None=Multiply(A=@n4) {constB=0.2})) {constB=0.04}
  n6 = Constant() {r=0.01}
  n7 = MaterialFunctionCall(UVs=AppendVector(A=Add(A=@n2) {constB=0.45}, B=Add(A=@n3, B=@n5) {constB=0}), Size=Constant() {r=0.04}, OutlineThickness=@n6) {materialFunction=&MF_UI_Circle}
  n8 = TextureCoordinate()
  n9 = ComponentMask(None=@n8) {r=true}
  n10 = Add(A=ComponentMask(None=@n8) {g=true}) {constB=0.1}
  n11 = Time()
  n12 = LinearInterpolate(Alpha=Sine(None=Multiply(A=@n11) {constB=0.25})) {constB=0.07}
  n13 = MaterialFunctionCall(UVs=AppendVector(A=Add(A=@n9) {constB=-0.45}, B=Add(A=@n10, B=@n12) {constB=0}), Size=Constant() {r=0.04}, OutlineThickness=@n6) {materialFunction=&MF_UI_Circle}
  n14 = TextureCoordinate()
  n15 = ComponentMask(None=@n14) {r=true}
  n16 = Add(A=ComponentMask(None=@n14) {g=true}) {constB=0.4}
  n17 = Time()
  n18 = LinearInterpolate(Alpha=Sine(None=Multiply(A=@n17) {constB=0.3})) {constB=0.05}
  n19 = MaterialFunctionCall(UVs=AppendVector(A=Add(A=@n15) {constB=-0.14}, B=Add(A=@n16, B=@n18) {constB=0}), Size=Constant() {r=0.03}, OutlineThickness=@n6) {materialFunction=&MF_UI_Circle}
  NamedRerouteDeclaration(None=Add(A=Add(A=@n7.Outline, B=@n13.Outline), B=@n19.Outline)) {name=Dots Outline, nodeColor=<rgba 1 0.8672 0 1>}
  NamedRerouteDeclaration(None=Add(A=Add(A=@n7.Fill, B=@n13.Fill), B=@n19.Fill)) {name=Dots Shape, nodeColor=<rgba 0 0 0 1>}
  n20 = NamedRerouteUsage()
  n21 = NamedRerouteUsage()
  NamedRerouteDeclaration(None=Multiply(A=LinearInterpolate(A=@n20, B=@n21, Alpha=@n21), B=ScalarParameter() {defaultValue=1, parameterName=Opacity})) {name=Opacity, nodeColor=<rgba 0.5703 0 1 1>}
  NamedRerouteUsage()
  NamedRerouteUsage()
 ]
}
