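# Translucent UI material for the low-gravity alert icon: a character over four concentric, purple-tinted circles with outlines.
# MF_UI_Circle builds each fill and outline mask, with a shared Scale Factor offsetting all sizes.
# Two panning Perlin-noise chains provide distorted UVs that make the shapes wobble, while a sine of Time bobs the center-scaled character texture vertically.
# Masks and icon alpha are lerped into Final Color and Opacity through named reroutes.
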Material /WildEstate_LowGrav/UI/Materials/M_UI_LowGrav_Alert_Icon
graph {
 properties {bAutomaticallySetUsageInEditor=false, blendMode=BLEND_Translucent, materialDomain=MD_UI, usedShadingModels=Default Lit}
 nodes [
  n1 = TextureCoordinate()
  n2 = ScalarParameter() {defaultValue=-0.15, parameterName=SpeedX}
  n3 = ScalarParameter() {defaultValue=0.15, parameterName=SpeedY}
  n4 = TextureSampleParameter2D(UVs=Panner(Coordinate=TextureCoordinate(), Speed=AppendVector(A=@n2, B=@n3)) {speedX=-1, speedY=2}) {parameterName=NoiseTexture, texture=/WildEstate_LowGrav/UI/Textures/T_UI_RTL_PerlinNoise3}
  n5 = AppendVector(A=@n4.R, B=@n4.R)
  n6 = ScalarParameter() {defaultValue=0.12, parameterName=Intensity}
  NamedRerouteDeclaration(None=Add(A=Subtract(A=@n1) {constB=0.02}, B=Multiply(A=@n5, B=@n6) {constB=0.056})) {name=Distort FX, nodeColor=<rgba 0 0 0 1>}
  n7 = ScalarParameter() {parameterName=Scale Factor}
  n8 = MaterialFunctionCall(
   UVs=NamedRerouteUsage()
   Size=Add(A=ScalarParameter() {defaultValue=0.6, parameterName=Size_Circle_1}, B=@n7) {constB=0}
   OutlineThickness=Constant() {r=0.015}
  ) {materialFunction=&MF_UI_Circle}
  NamedRerouteDeclaration(None=@n8.Fill) {name=Circle_1, nodeColor=<rgba 0 0 0 1>}
  n9 = NamedRerouteUsage()
  n10 = LinearInterpolate(
   A=Multiply(A=VectorParameter() {parameterName=BackroundColor_4, defaultValue=<rgba 0.4325 0.151 1 1>}.RGB, B=NamedRerouteUsage())
   B=Multiply(A=VectorParameter() {parameterName=BackroundColor_1, defaultValue=<rgba 0.4549 0.1823 1 1>}.RGB, B=@n9)
   Alpha=@n9
  )
  n11 = NamedRerouteUsage()
  n12 = Multiply(A=VectorParameter() {parameterName=BackroundColor_2, defaultValue=<rgba 0.2558 0.0478 0.6719 1>}.RGB, B=@n11)
  n13 = VectorParameter() {parameterName=BackroundColor_3, defaultValue=<rgba 0.0941 0.0205 0.2413 1>}
  n14 = NamedRerouteUsage()
  NamedRerouteDeclaration(None=LinearInterpolate(A=LinearInterpolate(A=@n10, B=@n12, Alpha=@n11), B=Multiply(A=@n13.RGB, B=@n14), Alpha=@n14)) {name=Circle Colors, nodeColor=<rgba 0 0 0 1>}
  n15 = MaterialFunctionCall(
   UVs=NamedRerouteUsage()
   Size=Add(A=ScalarParameter() {defaultValue=0.74, parameterName=Size_Circle_4}, B=@n7) {constB=0}
   OutlineThickness=Constant() {r=0.015}
  ) {materialFunction=&MF_UI_Circle}
  NamedRerouteDeclaration(None=@n15.Outline) {name=Outline_4, nodeColor=<rgba 0 0 0 1>}
  n16 = MaterialFunctionCall(
   UVs=NamedRerouteUsage()
   Size=Add(A=ScalarParameter() {defaultValue=0.3, parameterName=Size_Circle_3}, B=@n7) {constB=0}
   OutlineThickness=Constant() {r=0.015}
  ) {materialFunction=&MF_UI_Circle}
  NamedRerouteDeclaration(None=@n16.Fill) {name=Circle_3, nodeColor=<rgba 0 0 0 1>}
  NamedRerouteDeclaration(None=@n16.Outline) {name=Outline_3, nodeColor=<rgba 0 0 0 1>}
  n17 = NamedRerouteUsage()
  n18 = VectorParameter() {parameterName=OutlineColor_4, defaultValue=<rgba 0.4395 0.1615 1 1>}
  n19 = NamedRerouteUsage()
  n20 = LinearInterpolate(
   A=LinearInterpolate(
    A=LinearInterpolate(A=@n17, B=@n18.RGB, Alpha=@n19)
    B=VectorParameter() {parameterName=OutlineColor_1, defaultValue=<rgba 0.0682 0.0099 0.1163 1>}.RGB
    Alpha=NamedRerouteUsage()
   )
   B=VectorParameter() {parameterName=OutlineColor_2, defaultValue=<rgba 0.1661 0 0.4983 1>}.RGB
   Alpha=NamedRerouteUsage()
  )
  n21 = VectorParameter() {parameterName=OutlineColor_3, defaultValue=<rgba 0.0804 0 0.2413 1>}
  n22 = NamedRerouteUsage()
  NamedRerouteDeclaration(None=LinearInterpolate(A=LinearInterpolate(A=@n20, B=@n21.RGB, Alpha=@n22), B=NamedRerouteUsage(), Alpha=NamedRerouteUsage())) {name=Final Color, nodeColor=<rgba 0 0.7891 1 1>}
  n23 = TextureCoordinate()
  n24 = Constant() {r=-0.05}
  n25 = Constant() {r=0.05}
  n26 = TextureSampleParameter2D(UVs=Panner(Coordinate=TextureCoordinate(), Speed=AppendVector(A=@n24, B=@n25)) {speedX=-1, speedY=2}) {parameterName=NoiseTexture, texture=&T_UI_RTL_PerlinNoise3}
  n27 = AppendVector(A=@n26.R, B=@n26.R)
  n28 = ScalarParameter() {defaultValue=0.1, parameterName=Intensity 2}
  NamedRerouteDeclaration(None=Add(A=Subtract(A=@n23) {constB=0.02}, B=Multiply(A=@n27, B=@n28) {constB=0.2})) {name=Distort FX 2, nodeColor=<rgba 0.1484 1 0 1>}
  NamedRerouteDeclaration(None=@n8.Outline) {name=Outline_1, nodeColor=<rgba 0 0 0 1>}
  n29 = TextureCoordinate()
  n30 = Add(A=ComponentMask(None=@n29) {r=true}) {constB=0}
  n31 = Multiply(A=Time()) {constB=0.4}
  n32 = Add(A=Add(A=ComponentMask(None=@n29) {g=true}) {constB=0.06}, B=LinearInterpolate(Alpha=Sine(None=@n31)) {constB=0.05}) {constB=0}
  n33 = TextureSample(UVs=MaterialFunctionCall(UVs=AppendVector(A=@n30, B=@n32), Texture Scale=ScalarParameter() {defaultValue=0.9, parameterName=IconScale_Character}) {materialFunction=&ScaleUVsByCenter}.UVs) {samplerSource=SSM_Clamp_WorldGroupSettings, texture=/WildEstate_LowGrav/UI/Textures/T_UI_LowGrav_Character}
  NamedRerouteDeclaration(None=@n33.RGB) {name=IconColor, nodeColor=<rgba 0 0 0 1>}
  NamedRerouteDeclaration(None=@n33.A) {name=IconOpacity, nodeColor=<rgba 0 0 0 1>}
  n34 = NamedRerouteUsage()
  n35 = LinearInterpolate(A=Multiply(B=NamedRerouteUsage()) {constA=0.5}, B=@n34, Alpha=@n34)
  n36 = NamedRerouteUsage()
  n37 = LinearInterpolate(A=NamedRerouteUsage(), B=@n36, Alpha=@n36)
  n38 = NamedRerouteUsage()
  n39 = NamedRerouteUsage()
  n40 = Add(A=Add(A=Multiply(A=@n35), B=Multiply(B=@n37) {constA=0.5}), B=Multiply(A=LinearInterpolate(A=@n38, B=@n39, Alpha=@n39)))
  n41 = NamedRerouteUsage()
  n42 = LinearInterpolate(A=NamedRerouteUsage(), B=@n41, Alpha=@n41)
  n43 = NamedRerouteUsage()
  NamedRerouteDeclaration(None=LinearInterpolate(A=Add(A=@n40, B=@n42), B=@n43, Alpha=@n43)) {name=Opacity, nodeColor=<rgba 0.5703 0 1 1>}
  NamedRerouteDeclaration(None=@n15.Fill) {name=Circle_4, nodeColor=<rgba 0 0 0 1>}
  n44 = MaterialFunctionCall(
   UVs=NamedRerouteUsage()
   Size=Add(A=ScalarParameter() {defaultValue=0.45, parameterName=Size_Circle_2}, B=@n7) {constB=0}
   OutlineThickness=Constant() {r=0.015}
  ) {materialFunction=&MF_UI_Circle}
  NamedRerouteDeclaration(None=@n44.Outline) {name=Outline_2, nodeColor=<rgba 0 0 0 1>}
  NamedRerouteDeclaration(None=@n44.Fill) {name=Circle_2, nodeColor=<rgba 0 0 0 1>}
  NamedRerouteUsage()
  NamedRerouteUsage()
 ]
}
